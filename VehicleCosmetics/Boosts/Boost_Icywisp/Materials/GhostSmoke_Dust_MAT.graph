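# Additive, unlit, two-sided Niagara sprite material for ghostly smoke dust in the Icywisp vehicle boost effect.
# Two panning flow-map samples, scaled by the tile parameter, distort the UVs of a scrolling water noise texture.
# The speed parameter drives all panning, while distortion strength is constant despite a named distort parameter.
# The noise is shaped by a flame-cloud mask, and particle color feeds an emissive-adjustment function.
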Material /VehicleCosmetics/Boosts/Boost_Icywisp/Materials/GhostSmoke_Dust_MAT
graph {
 properties {
  bUsedWithNiagaraSprites=true
  blendMode=BLEND_Additive
  pixelDepthOffsetMode=PDOM_Legacy
  shadingModel=MSM_Unlit
  twoSided=true
  usedShadingModels=Unlit
 }
 nodes [
  MaterialFunctionCall(Emissive=ParticleColor().RGB) {materialFunction=&MF_FX_EmissiveAdjustments}
  n1 = DynamicParameter() {paramNames [tile,speed,distort,Param4]}
  n2 = Multiply(A=@n1.speed, B=Time())
  n3 = Panner(Coordinate=TextureCoordinate(), Time=@n2) {speedX=0.05, speedY=0.1}
  n4 = TextureCoordinate() {uTiling=0.75, vTiling=0.75}
  n5 = TextureSample(UVs=Panner(Coordinate=Multiply(A=@n4, B=@n1.tile), Time=@n2) {speedX=0.05, speedY=0.1}) {texture=/DealershipSharedLibrary/FX/Textures/Noise/T_VCVFX_Flow03_N}
  n6 = TextureCoordinate() {uTiling=0.5, vTiling=0.5}
  n7 = TextureSample(UVs=Panner(Coordinate=Multiply(A=@n6, B=@n1.tile), Time=@n2) {speedX=-0.03, speedY=0.01}) {texture=&T_VCVFX_Flow03_N}
  n8 = Multiply(A=Constant() {r=0.3}, B=ComponentMask(None=Add(A=@n5.RGB, B=@n7.RGB)) {g=true, r=true})
  n9 = Power(Base=TextureSample(UVs=Add(A=@n3, B=@n8)) {texture=/DealershipSharedLibrary/FX/Textures/Noise/T_VCVFX_Water01_A}.RGB, Exp=Constant() {r=1})
  n10 = Constant() {r=1.5}
  Multiply(B=Multiply(A=TextureSample() {texture=/DealershipSharedLibrary/FX/Textures/Gradient_Masks/T_VCVFX_FlameCloud_02}.RGB, B=Multiply(A=@n9, B=@n10)))
 ]
}
