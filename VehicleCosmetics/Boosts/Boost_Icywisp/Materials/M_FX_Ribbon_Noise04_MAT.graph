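# Translucent, unlit, two-sided Niagara ribbon and sprite material for wispy, scrolling smoke in the Icywisp vehicle boost effect.
# Flow and panning curl noise distort cone-remapped, scrolling UVs that sample a wispy smoke mask, with particle alpha driving its contrast.
# A view fresnel using a face normal from world-position ddx/ddy, UV edge masks and emitter-life opacity shape the alpha.
# Particle color and dynamic EmissiveMult feed a shared emissive adjustment function.
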Material /VehicleCosmetics/Boosts/Boost_Icywisp/Materials/M_FX_Ribbon_Noise04_MAT
graph {
 properties {
  bUsedWithNiagaraRibbons=true
  bUsedWithNiagaraSprites=true
  blendMode=BLEND_Translucent
  pixelDepthOffsetMode=PDOM_Legacy
  shadingModel=MSM_Unlit
  twoSided=true
  usedShadingModels=Unlit
 }
 nodes [
  MaterialExpressionWorldPosition_0 = WorldPosition()
  n1 = Custom!(x=@MaterialExpressionWorldPosition_0.XYZ) {
   code="return ddx(x);"
   inputs [{
     inputName=x
     input {
      expression=&self:MaterialExpressionWorldPosition_0
      mask=1
      maskB=1
      maskG=1
      maskR=1
     }
    }]
  }
  n2 = Custom!(y=@MaterialExpressionWorldPosition_0.XYZ) {
   code="return ddy(y);"
   inputs [{
     inputName=y
     input {
      expression=&self:MaterialExpressionWorldPosition_0
      mask=1
      maskB=1
      maskG=1
      maskR=1
     }
    }]
  }
  n3 = Normalize(VectorInput=CrossProduct(A=@n1, B=@n2))
  n4 = Multiply(A=Transform(None=Constant3Vector() {constant=<rgba 0 0 1 0>}) {transformSourceType=TRANSFORMSOURCE_View}, B=Constant() {r=-1})
  n5 = TextureCoordinate()
  n6 = ConstantBiasScale(None=TextureSample(UVs=@n5) {texture=/DealershipSharedLibrary/FX/Textures/Noise/T_VCVFX_Flow_Push_01}.RGBA)
  n7 = ScalarParameter() {parameterName=Static DuDv Mult}
  n8 = ComponentMask(None=VectorParameter() {parameterName=Smoke DuDv Tile // Scroll, defaultValue=<rgba 1.3 0.5 -3 0.1>}.RGB) {g=true, r=true}
  n9 = TextureSample(UVs=Panner(Coordinate=Multiply(A=@n5, B=@n8), Time=Time()) {speedX=-3, speedY=0.01}) {texture=/DealershipSharedLibrary/FX/Textures/Noise/T_VCVFX_Flow_Curl_01}
  n10 = Multiply(
   A=Multiply(A=ConstantBiasScale(None=@n9.RGB), B=ScalarParameter() {defaultValue=0.3, parameterName=SmokeDuDvMul})
   B=DynamicParameter() {paramNames [Opacity/Life, Param2, DuDvMultiplier, EmissiveMult]}.DuDvMultiplier
  )
  n11 = TextureCoordinate()
  n12 = ComponentMask(None=@n11) {r=true}
  n13 = Time()
  n14 = Add(
   A=Multiply(A=@n12, B=ScalarParameter() {defaultValue=2, parameterName=Beam Noise Tile Y})
   B=Multiply(A=@n13, B=ScalarParameter() {defaultValue=-4.2, parameterName=Beam Noise Pan Speed X})
  )
  n15 = Constant() {r=0.5}
  n16 = ComponentMask(None=@n11) {g=true}
  n17 = MaterialFunctionCall(
   Input=@n12
   Input High=ScalarParameter() {defaultValue=1, parameterName=OldMaxU}
   Target Low=ScalarParameter() {defaultValue=0.2, parameterName=Cone Width Top}
   Target High=ScalarParameter() {defaultValue=1, parameterName=Cone Width Bottom}
  ) {materialFunction=&RemapValueRange}
  n18 = Constant() {r=-1}
  n19 = Add(A=@n15, B=Multiply(A=Subtract(A=@n16, B=@n15), B=Power(Base=@n17.Result, Exp=@n18)))
  n20 = Constant()
  n21 = Constant() {r=1}
  n22 = ScalarParameter() {parameterName=Beam Noise Pan Speed Y}
  n23 = Multiply(
   A=Add(A=Clamp(None=@n19, Min=@n20, Max=@n21), B=Multiply(A=@n22, B=@n13))
   B=ScalarParameter() {defaultValue=1, parameterName=Noise Bottom Squash}
  )
  n24 = Add(
   A=ComponentMask(None=Multiply(A=@n6, B=@n7)) {g=true, r=true}
   B=Add(A=ComponentMask(None=@n10) {g=true, r=true}, B=AppendVector(A=@n14, B=@n23))
  )
  n25 = AppendVector(
   A=Add(A=ScalarParameter() {defaultValue=0.1, parameterName=Smoke U Offset}, B=ComponentMask(None=@n24) {r=true})
   B=Saturate(None=ComponentMask(None=@n24) {g=true})
  )
  n26 = ParticleColor()
  n27 = Multiply(A=ScalarParameter() {defaultValue=5, parameterName=Smoke Opacity Power}, B=OneMinus(None=@n26.A))
  n28 = ScalarParameter() {defaultValue=0.4, parameterName=Smoke Opacity Power Offset}
  n29 = Multiply(
   A=Power(Base=DotProduct(A=@n3, B=@n4), Exp=ScalarParameter() {defaultValue=2, parameterName=Fresnel Power})
   B=Power(
    Base=TextureSample(UVs=@n25) {texture=/DealershipSharedLibrary/FX/Textures/Gradient_Masks/T_VCVFX_Smoke_Wispy_Circle_03_A}.R
    Exp=Add(A=@n27, B=@n28)
   )
  )
  ConstantBiasScale(None=@n29) {bias=150}
  Constant() {r=5}
  n30 = DynamicParameter() {paramNames [OpacityOverEmitterLife, Parameter1, DuDv Multiplier, EmissiveMult]}
  MaterialFunctionCall(Emissive=Multiply(A=@n30.EmissiveMult, B=ParticleColor().RGBA)) {materialFunction=&MF_FX_EmissiveAdjustments}
  Multiply(A=@n29)
  n31 = Multiply(A=@n29, B=@n26.A)
  n32 = TextureCoordinate()
  n33 = ComponentMask(None=@n32) {g=true}
  n34 = Multiply(A=OneMinus(None=@n33), B=@n33)
  n35 = ComponentMask(None=@n32) {r=true}
  n36 = OneMinus(None=@n35)
  n37 = Multiply(
   A=Power(Base=Multiply(A=@n34, B=@n36), Exp=ScalarParameter() {defaultValue=1.25, parameterName=EdgeMaskPower})
   B=ScalarParameter() {defaultValue=8, parameterName=EdgeMaskMult}
  )
  n38 = Multiply(
   A=Power(Base=@n35, Exp=ScalarParameter() {defaultValue=0.75, parameterName=StartEdgeMaskPower})
   B=ScalarParameter() {defaultValue=4, parameterName=StartEdgeMaskMult}
  )
  n39 = Multiply(A=@n30.OpacityOverEmitterLife, B=Multiply(A=Multiply(A=@n31, B=@n37), B=Saturate(None=@n38)))
  Multiply(A=@n39, B=DepthFade())
  Saturate(None=@n39)
  ScalarParameter() {parameterName=SmokeDuDv}
 ]
}
