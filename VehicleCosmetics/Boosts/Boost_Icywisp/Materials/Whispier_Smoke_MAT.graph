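# Translucent Niagara ribbon material for wispy smoke in the Icywisp vehicle boost effect.
# Panning cloud noise distorts scrolling UVs, offset in U by a dynamic parameter, that sample a wispy smoke mask multiplied by a second noise.
# The result is tinted by particle color through MF_FX_EmissiveAdjustments.
# An alpha term combines the same mask with particle alpha and a power-curve fade at both U ends of the ribbon.
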Material /VehicleCosmetics/Boosts/Boost_Icywisp/Materials/Whispier_Smoke_MAT
graph {
 properties {bUsedWithNiagaraRibbons=true, blendMode=BLEND_Translucent, pixelDepthOffsetMode=PDOM_Legacy, usedShadingModels=Default Lit}
 nodes [
  n1 = ParticleColor()
  n2 = Time()
  n3 = Constant() {r=1}
  n4 = TextureSample(UVs=Panner(Coordinate=TextureCoordinate() {uTiling=0.35, vTiling=0.35}, Time=Multiply(A=@n2, B=@n3)) {speedX=0.1, speedY=0.1}) {texture=/DealershipSharedLibrary/FX/Textures/Noise/T_VCVFX_CloudsSoft_RGB}
  n5 = TextureCoordinate()
  n6 = AppendVector(A=DynamicParameter() {paramNames [Offset,Param2,Param3,Param4]}.Offset, B=Constant())
  n7 = Add(
   A=Multiply(A=ComponentMask(None=@n4.RGB) {g=true, r=true}, B=ScalarParameter() {defaultValue=0.1, parameterName=Distortion})
   B=Panner(Coordinate=Add(A=@n5, B=@n6)) {speedX=-1.5}
  )
  n8 = Multiply(
   A=TextureSample(UVs=@n7) {texture=/DealershipSharedLibrary/FX/Textures/Gradient_Masks/T_VCVFX_Wispier_Smoke_A}.RGB
   B=TextureSample(UVs=@n7) {texture=/DealershipSharedLibrary/FX/Textures/Noise/T_VCVFX_Noise1_A}.G
  )
  MaterialFunctionCall(Emissive=Multiply(A=@n1.RGB, B=@n8)) {materialFunction=&MF_FX_EmissiveAdjustments}
  n9 = ComponentMask(None=TextureCoordinate()) {r=true}
  n10 = Constant() {r=1.2}
  n11 = OneMinus(None=@n9)
  n12 = Constant() {r=1}
  Multiply(A=Multiply(A=@n8, B=@n1.A), B=Multiply(A=Power(Base=@n9, Exp=@n10), B=Power(Base=@n11, Exp=@n12)))
 ]
}
