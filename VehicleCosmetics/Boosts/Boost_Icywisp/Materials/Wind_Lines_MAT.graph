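# Translucent, two-sided Niagara ribbon material that draws streaking wind lines for the Icywisp vehicle boost effect.
# A scrolling wind-line texture has its UVs distorted by panning cloud noise, scaled by a Distortion parameter.
# Its red channel is tinted by particle color and passed through a shared emissive adjustment function.
# The same channel, multiplied by particle alpha and a U-axis power gradient that fades both ribbon ends, forms an opacity mask.
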Material /VehicleCosmetics/Boosts/Boost_Icywisp/Materials/Wind_Lines_MAT
graph {
 properties {bUsedWithNiagaraRibbons=true, blendMode=BLEND_Translucent, pixelDepthOffsetMode=PDOM_Legacy, twoSided=true, usedShadingModels=Default Lit}
 nodes [
  n1 = ParticleColor()
  n2 = TextureCoordinate() {uTiling=0.35, vTiling=0.35}
  n3 = Multiply(A=Time(), B=Constant() {r=1})
  n4 = ComponentMask(None=TextureSample(UVs=Panner(Coordinate=@n2, Time=@n3) {speedX=0.1, speedY=0.1}) {texture=/DealershipSharedLibrary/FX/Textures/Noise/T_VCVFX_CloudsSoft_RGB}.RGB) {g=true, r=true}
  n5 = ScalarParameter() {defaultValue=0.1, parameterName=Distortion}
  n6 = TextureCoordinate() {uTiling=2}
  n7 = TextureSample(UVs=Add(A=Multiply(A=@n4, B=@n5), B=Panner(Coordinate=@n6) {speedX=-3})) {texture=/DealershipSharedLibrary/FX/Textures/Gradient_Masks/T_VCVFX_Wind_Lines_RGB}
  MaterialFunctionCall(Emissive=Multiply(A=@n1.RGB, B=@n7.R)) {materialFunction=&MF_FX_EmissiveAdjustments}
  n8 = ComponentMask(None=TextureCoordinate()) {r=true}
  n9 = Constant() {r=1.2}
  n10 = OneMinus(None=@n8)
  n11 = Constant() {r=1}
  Multiply(A=Multiply(A=@n7.R, B=@n1.A), B=Multiply(A=Power(Base=@n8, Exp=@n9), B=Power(Base=@n10, Exp=@n11)))
 ]
}
